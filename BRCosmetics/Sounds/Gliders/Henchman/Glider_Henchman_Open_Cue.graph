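# Sound cue for the Henchman glider cosmetic's opening sound.
# A mixer plays two waves together, the open effect and a music sting, with the sting weighted louder.
# Playback uses the player glider open/close attenuation and concurrency, the cosmetic glider-open property sheet and a glider sound class named for landing.
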SoundCue /BRCosmetics/Sounds/Gliders/Henchman/Glider_Henchman_Open_Cue
root=Mixer_0
nodes [
 {id=Mixer_0, children [WavePlayer_1,WavePlayer_0], properties {inputVolume [0.9,1.4]}}
 {id=WavePlayer_0, properties {soundWaveAssetPtr=&/Glider_Henchman_Open__Mx_Sting}}
 {id=WavePlayer_1, properties {soundWaveAssetPtr=&/Glider_Henchman_Open_01}}
]
properties {
 attenuationSettings=/Game/Sounds/Attenuation/Fort_Player_Glider_OpenClose_Athena_Att
 audioPropertiesSheet=/BRCosmetics/Sounds/Settings/PropertySheets/Glider/APS_Cosmetic_Glider_Open
 maxDistance=10000
 soundClassObject=&Fort_Gliders_Land
 volumeMultiplier=1
 concurrencySet [&Athena_GliderOpenClose_Conc]
}
